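# Niagara swing effect for the female Forsake pickaxe, drawing ribbon trails and glowing lights on its skeletal mesh.
# Trail001 spawns ribbon particles from mesh sampling for three renderers, with distance-based UVs, sampled-rotation facing and fading alpha.
# Lights pins three sprite-and-light particles to selected bones or sockets.
# An alt boolean, apparently derived from User.IsAlt, switches both emitters' color schemes; the older AnimTrail-based Trail emitter is disabled.
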NiagaraSystem /BRCosmetics/Weapons/FORT_Melee/Pickaxe_Forsake_Female/FX/NS_Forsake_FemaleSwing
dependencies {
 usedDataInterfaces [&NiagaraDataInterfaceCurve, &NiagaraDataInterfaceArrayInt32]
 usedDynamicInputs [&SetBoolByFloatComparison]
 usedModules [
  &self:SystemSpawnScript.SystemScriptSource.SystemScriptGraph.NiagaraNodeAssignment_0.SetVariables_#1
  &SystemState
  &Timeline
 ]
}
emitters {
 Lights {
  emitterUpdateScript [
   &EmitterState()
   &SpawnBurst_Instantaneous(Age:NiagaraFloat=$Emitter.Age, Spawn Count=3, Use Spawn Probability=false, UseLoopCountLimit=false)
  ]
  particleSpawnScript [&InitializeParticle(
    Color=&MultiplyLinearColor(
     Linear Color=&MakeCustomLinearColorFromBool(Bool=$System.IsAltBool, True Color=<rgba 0 0 1 1>, False Color=<rgba 1 0 0.7999 1>)
     Scale Factor=<1 1 1 1>
    )
    Lifetime=5
    Sprite Size=<22 33>
    Sprite Size Mode=::NewEnumerator4
    UsePositionOffset=false
    Write Material Random=true
   )]
  particleUpdateScript [
   &ParticleState(
    Kill Particles When Lifetime Has Elapsed=false
    Let Infinitely Lived Particles Die When Emitter Deactivates=true
    Loop Particles Lifetime=true
   )
   &SkeletalMeshLocation(
    Bone / Socket Index=&SelectIntFromArray(
     Direct Array Index=&ReturnExecIndex()
     Int Selection Array=DataInterface() {propertyValues {IntData [4,10,16]}}
    )
    Bone Sampling Mode=::NewEnumerator4
    MaskSpawnGroup=false
    Skeletal Mesh=$User.SkeletalMesh
   )
   &FN_Light_Attributes(
    Color=&HueShiftLinearColor(Hue=0, Saturation=0.5, Value=1, Linear Color=$Particles.Color)
    Create Color Attribute=true
    Scale Color=3
   )
   /SetVariables(Particles.Sprite COlor:LinearColor=&MultiplyLinearColor(Linear Color=$Particles.Color, Scale Factor=&MakeVector4(X=1, Y=1, Z=1, W=0.5))) {isSetParameters=true}
   /SetVariables(Particles.Velocity=&ReturnMeshOrientationAxis(Mesh Orientation Axis=::XAxis, Mesh Orientation Quat=$Particles.SkeletalMeshLocation.SampledRotation)) {isSetParameters=true}
   &CameraOffset(Camera Offset Amount=2)
  ]
  rendererClasses [&NiagaraSpriteRendererProperties, &NiagaraLightRendererProperties]
  renderers [
   {rendererClass=&NiagaraSpriteRendererProperties}
   {rendererClass=&NiagaraLightRendererProperties, rendererIndex=1}
  ]
 }
 Trail {
  bEnabled=false
  emitterUpdateScript [
   &EmitterState() {module=EmitterState001}
   &FN_SpawnAnimTrail_v2(Spawn Interval=2)
   &self:CalcDistance_Travelled(Velocity Vector:Vector3f=&Multiply_Vector(A=$Engine.Owner.Velocity, B=&VectorFromFloat(Value=-1))) {module=CalcDistance_Travelled}
  ]
  particleSpawnScript [
   /Game/Effects/Niagara/Modules/AnimTrail/FN_InitializeAnimTrail/FN_InitializeAnimTrailBoneIndecies()
   /Game/Effects/Niagara/Modules/AnimTrail/FN_InitializeAnimTrail/FN_InitializeAnimTrail(
    AnimTrailWidth=1.5
    Color=&MakeCustomLinearColorFromBool(Bool=true, True Color=<rgba 1 1 1 1>, False Color=<rgba 1 1 1 1>)
    Lifetime=0.3
   )
   /SetVariables(Particles.RibbonUVDistance:NiagaraFloat=$Emitter.DistanceTraveled) {isSetParameters=true, module=SetVariables_#2}
   &AddVelocity(Velocity=<0 1500 50>) {enabled=false}
   /SetVariables(Particles.RibbonWidth=25) {isSetParameters=true, module=SetVariables_#3}
  ]
  particleUpdateScript [
   &SolveForcesAndVelocity(Clamp Velocity=false, Limit Acceleration=false) {enabled=false}
   &FN_UpdateAnimTrail_v2()
   &ScaleRibbonWidth(Ribbon Width Scale=&FloatFromCurve(FloatCurve=DataInterface() {propertyValues {Curve {keys [
         {
          arriveTangent=0.5623
          arriveTangentWeight=0.4591
          interpMode=RCIM_Cubic
          leaveTangent=0.5623
          leaveTangentWeight=0.4591
          tangentMode=RCTM_User
          tangentWeightMode=RCTWM_WeightedBoth
          time=0
          value=1
         }
         {
          arriveTangent=-0.2443
          arriveTangentWeight=0.7082
          interpMode=RCIM_Cubic
          leaveTangent=-0.2443
          leaveTangentWeight=0.7082
          tangentMode=RCTM_User
          tangentWeightMode=RCTWM_WeightedBoth
          time=1
          value=0
         }
        ]}}})) {enabled=false}
   &ScaleColor(
    Scale Alpha=&Multiply_Float(
     A=&FloatFromCurve(FloatCurve=DataInterface() {propertyValues {Curve {keys [
          {
           arriveTangent=-3.175
           arriveTangentWeight=0.9161
           interpMode=RCIM_Cubic
           leaveTangent=-3.175
           leaveTangentWeight=0.9161
           tangentMode=RCTM_User
           tangentWeightMode=RCTWM_WeightedBoth
           time=0
           value=1
          }
          {
           arriveTangent=-0.0544
           arriveTangentWeight=0.581
           interpMode=RCIM_Cubic
           leaveTangent=-0.3338
           leaveTangentWeight=0.2826
           tangentMode=RCTM_Break
           tangentWeightMode=RCTWM_WeightedBoth
           time=1
           value=0
          }
         ]}}})
     B=&FloatFromCurve(
      FloatCurve=DataInterface() {propertyValues {Curve {keys [
          {
           interpMode=RCIM_Cubic
           leaveTangent=6.4657
           leaveTangentWeight=1.1441
           tangentMode=RCTM_Break
           tangentWeightMode=RCTWM_WeightedBoth
           time=0
           value=0
          }
          {interpMode=RCIM_Cubic, tangentMode=RCTM_Break, time=1, value=1}
         ]}}}
      CurveIndex=/Game/Effects/Niagara/DynamicInputs/Math/FN_NormalizedParticleCount()
      Scale Curve=2
     )
    )
    ScaleA=true
    ScaleRGB=false
   )
  ]
  rendererClasses [&NiagaraRibbonRendererProperties]
  renderers [{rendererClass=&NiagaraRibbonRendererProperties}]
 }
 Trail001 {
  emitterUpdateScript [
   &EmitterState(
    Enable Visibility Culling=true
    Scalability Mode=::NewEnumerator1
    Visibility Cull Delay=0.01
    Visibility Cull Response:ENiagaraExecutionStateManagement=::SleepAndClearParticles
   ) {module=EmitterState001}
   &self:CalcDistance_Travelled(Velocity Vector:Vector3f=&Multiply_Vector(A=$Engine.Owner.Velocity, B=&VectorFromFloat(Value=-1))) {module=CalcDistance_Travelled}
   &SpawnRate(SpawnRate=120, Use Spawn Probability=false)
  ]
  particleSpawnScript [
   &InitializeRibbon(
    Color=&MakeCustomLinearColorFromBool(Bool=$System.IsAltBool, True Color=<rgba 0.3001 0.3001 1 1.5>, False Color=<rgba 1 0.25 0 1>)
    Lifetime=0.15
    UsePositionOffset=false
    Write Material Random=true
   )
   /SetVariables(Particles.RibbonUVDistance:NiagaraFloat=$Emitter.DistanceTraveled) {isSetParameters=true, module=SetVariables_#2}
   /SetVariables(Particles.RibbonWidth=25) {isSetParameters=true, module=SetVariables_#3}
   &SkeletalMeshLocation(
    Interpolate Mesh Sampling=false
    MaskSpawnGroup=false
    Mesh Sampling Type=::NewEnumerator5
    Skeletal Mesh=$User.SkeletalMesh
    Socket Sampling Mode=::NewEnumerator5
   )
   /SetVariables(Particles.RibbonFacing=&ReturnMeshOrientationAxis(Mesh Orientation Axis=::XAxis, Mesh Orientation Quat=$Particles.SkeletalMeshLocation.SampledRotation)) {isSetParameters=true}
   /SetVariables(Particles.pink line:NiagaraFloat=&Multiply_Float(A=$Particles.RibbonWidth, B=0.75)) {isSetParameters=true}
  ]
  particleUpdateScript [
   &ParticleState()
   &ScaleColor(
    Scale Alpha=&Multiply_Float(
     A=&Multiply_Float(
      A=&FloatFromCurve(FloatCurve=DataInterface() {propertyValues {Curve {keys [
           {
            arriveTangent=-8.5427
            arriveTangentWeight=0.8208
            interpMode=RCIM_Cubic
            leaveTangent=-8.5427
            leaveTangentWeight=0.8208
            tangentMode=RCTM_User
            tangentWeightMode=RCTWM_WeightedBoth
            time=0.1
            value=1
           }
           {
            arriveTangent=-0.9101
            arriveTangentWeight=0.5943
            interpMode=RCIM_Cubic
            leaveTangent=-0.3338
            leaveTangentWeight=0.2826
            tangentMode=RCTM_Break
            tangentWeightMode=RCTWM_WeightedBoth
            time=0.6
            value=0
           }
          ]}}})
      B=&FloatFromCurve(
       FloatCurve=DataInterface() {propertyValues {Curve {keys [{interpMode=RCIM_Cubic, tangentMode=RCTM_Break, time=1, value=1}]}}}
       CurveIndex=&FN_NormalizedParticleCount()
       Scale Curve=2
      )
     )
     B=1
    )
    ScaleA=true
    ScaleRGB=false
   )
   /SetVariables(Particles.Line Color:LinearColor=&MultiplyLinearColor(
     Linear Color=&MakeCustomLinearColorFromBool(Bool=$System.IsAltBool, True Color=<rgba 0.25 0.375 1 0.2>, False Color=<rgba 0.9 0.25 1 0.2>)
     Scale Factor=&MakeFloat4FromVectorAndFloat(
      "Vector (RGB)"=<1 1 1>
      "Float (Alpha)"=&FloatFromCurve(FloatCurve=DataInterface() {propertyValues {Curve {keys [<at 0 1>,<at 0.317 1>,<at 0.367 0>]}}})
     )
    )) {isSetParameters=true}
  ]
  rendererClasses [&NiagaraRibbonRendererProperties]
  renderers [
   {rendererClass=&NiagaraRibbonRendererProperties}
   {rendererClass=&NiagaraRibbonRendererProperties, rendererIndex=1}
   {rendererClass=&NiagaraRibbonRendererProperties, rendererIndex=2}
  ]
 }
}
summary {
 systemName=NS_Forsake_FemaleSwing
 emitters [
  {bEnabled=false, emitterName=Trail, rendererClasses [&NiagaraRibbonRendererProperties]}
  {emitterName=Trail001, rendererClasses [&NiagaraRibbonRendererProperties]}
  {emitterName=Lights, rendererClasses [&NiagaraSpriteRendererProperties, &NiagaraLightRendererProperties]}
 ]
 userVariables [
  User.IsAlt {type=NiagaraFloat, defaultValue=0, description=""}
  User.RibbonIDOffset {type=NiagaraInt32}
  User.SkeletalMesh {
   type=NiagaraDataInterfaceSkeletalMesh
   defaultValue {
    struct=NiagaraExt_VariableValue_DataInterface
    value {dataInterface=&self:NiagaraDataInterfaceSkeletalMesh_0}
   }
  }
 ]
}
